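# Conversation database for The Scientist NPC, defining the greeting a player gets on starting a conversation and the options that follow.
# The entry point leads to one speech node with a default line, a first-ever-meeting line and skin-specific lines gated by character cosmetic ID.
# Priority and weight values rank the contextual lines.
# The speech node then links by entry tag to the shared NPC quest and basic services conversations.
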ConversationDatabase /NPCLibrary/NPCs/TheScientist/TheScientist/NPCLibrary_Conversation_TheScientist
entry_points [{tag=Conversation.Entry.TheScientist, destinations [#1]}]
nodes [
 ConversationEntryPointNode_7 {
  id=#1
  class=ConversationEntryPointNode
  outputs [#2]
  properties {entryTag {tagName=Conversation.Entry.TheScientist}}
 }
 Speech_C_4 {
  id=#2
  class=Speech_C
  outputs [#3, #4]
  properties {
   nodeName=""
   generalConfig {
    defaultMessage=What's gone wrong now?
    contextualMessages [
     {
      message="You lived! I'm... \"Surprised\" is the wrong word..."
      contextRequirements [{requirement {class=FortControllerRequirement_IsFirstEverConversationWithNPC}}]
      priority {curve {rowName=Tandem.Speech.Priority.FirstEverInteraction}}
     }
     {
      message=Our next threat won't be flying in on an airship.
      contextRequirements [{requirement {
         class=FortControllerRequirement_HasCID
         properties {allowedCIds [/BRCosmetics/Athena/Items/Cosmetics/Characters/CID_A_373_Athena_Commando_M_OriginPrisoner]}
        }}]
      priority {value=0, curve {curveTable=null, rowName=None}}
     }
     {
      message="These calculations aren't looking good. "
      contextRequirements [{requirement {
         class=FortControllerRequirement_HasCID
         properties {allowedCIds [/BRCosmetics/Athena/Items/Cosmetics/Characters/CID_547_Athena_Commando_F_Meteorwoman]}
        }}]
      priority {value=10, curve {curveTable=null, rowName=None}}
      weight {value=10}
     }
     {
      message=I owe you one for bringing her back.
      contextRequirements [{requirement {
         class=FortControllerRequirement_HasCID
         properties {allowedCIds [/BRCosmetics/Athena/Items/Cosmetics/Characters/CID_A_374_Athena_Commando_F_Binary]}
        }}]
      priority {value=0, curve {curveTable=null, rowName=None}}
     }
     {
      message="Aah! Don't sneak up on me like that! "
      contextRequirements [{requirement {
         class=FortControllerRequirement_HasCID
         properties {allowedCIds [/BRCosmetics/Athena/Items/Cosmetics/Characters/CID_A_395_Athena_Commando_F_BinaryTwin]}
        }}]
      priority {value=0, curve {curveTable=null, rowName=None}}
     }
     {
      message=Please don't ask me to fix the monument again.
      contextRequirements [{requirement {
         class=FortControllerRequirement_HasCID
         properties {allowedCIds [/BRCosmetics/Athena/Items/Cosmetics/Characters/CID_A_286_Athena_Commando_M_Turtleneck]}
        }}]
      priority {value=10, curve {curveTable=null, rowName=None}}
      weight {value=10}
     }
     {
      message=So you're thinking of settling down? I hear Sleepy's nice.
      contextRequirements [{requirement {
         class=FortControllerRequirement_HasCID
         properties {allowedCIds [/BRCosmetics/Athena/Items/Cosmetics/Characters/CID_140_Athena_Commando_M_Visitor]}
        }}]
      priority {value=0, curve {curveTable=null, rowName=None}}
     }
     {
      message=Don't you have someone to find?
      contextRequirements [{requirement {
         class=FortControllerRequirement_HasCID
         properties {allowedCIds [&CID_883_Athena_Commando_M_ChOneJonesy]}
        }}]
      priority {value=0, curve {curveTable=null, rowName=None}}
     }
    ]
   }
  }
 }
 ConversationLinkNode_0 {id=#4, class=ConversationLinkNode, properties {remoteEntryTag {tagName=Conversation.Entry.BasicServices}}}
 ConversationLinkNode_1 {id=#3, class=ConversationLinkNode, properties {remoteEntryTag {tagName=Conversation.Entry.NPCLibrary.Quests}}}
]
